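# Animation Blueprint for the TimberStakeClub male character's face accessory, adding secondary motion to its hood.
# A pose copied from the attached parent mesh passes through a Control Rig, then five chained spring-driven Anim Dynamics nodes.
# Front hood bones simulate relative to base bones, side and back ones relative to root.
# The event graph zeroes Phys_off during zipline or grind rail, but dynamics alphas stay at 1, not reading it.
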AnimBlueprint /BRCosmetics/Characters/Player/Male/Medium/Bodies/M_MED_TimberStakeClub/Meshes/Parts/M_MED_TimberStakeClub_FaceAcc_AnimBP
target_skeleton=&/M_MED_TimberStakeClub_FaceAcc_Skeleton
graphs [
 AnimGraph {nodes [
   n1 = Control Rig(Source=Copy Pose From Mesh() {bUseAttachedParent=true}) {
    alphaBoolBlend {blendOption=Linear}
    controlRigAssetReference {blueprintRigClass=&/M_MED_TimberStakeClub_FaceAcc_CtrlRig.M_MED_TimberStakeClub_FaceAcc_CtrlRig_C}
   }
   n2 = Anim Dynamics - Bone: dyn_hoodfront_l(ComponentPose=Local To Component(LocalPose=@n1), Alpha=1, ExternalForce="0,0,0") {
    angularDampingOverride=1
    angularSpringConstant=14000
    bAngularSpring=true
    bOverrideAngularDamping=true
    bUseGravityOverride=true
    gravityScale=0.35
    simulationSpace=BoneRelative
    alphaBoolBlend {blendOption=Linear}
    boundBone {boneName=dyn_hoodfront_l}
    componentLinearAccScale=<1 1 1>
    componentLinearVelScale=<1 1 1>
    gravityOverride {z=-980}
    physicsBodyDefinitions [{
      boundBone {boneName=dyn_hoodfront_l}
      boxExtents=<20 15 5>
      constraintSetup {angularLimitsMax=<20 0 20>, angularLimitsMin=<-20 0 -20>, angularTarget=<1 0 0>}
      localJointOffset=<4 0 0>
     }]
    relativeSpaceBone {boneName=base_hoodfront_l}
   }
   n3 = Anim Dynamics - Bone: dyn_hoodfront_r(ComponentPose=@n2, Alpha=1, ExternalForce="0,0,0") {
    angularDampingOverride=1
    angularSpringConstant=14000
    bAngularSpring=true
    bOverrideAngularDamping=true
    bUseGravityOverride=true
    gravityScale=0.35
    simulationSpace=BoneRelative
    alphaBoolBlend {blendOption=Linear}
    boundBone {boneName=dyn_hoodfront_r}
    componentLinearAccScale=<1 1 1>
    componentLinearVelScale=<1 1 1>
    gravityOverride {z=-980}
    physicsBodyDefinitions [{
      boundBone {boneName=dyn_hoodfront_r}
      boxExtents=<20 15 5>
      constraintSetup {angularLimitsMax=<20 0 20>, angularLimitsMin=<-20 0 -20>, angularTarget=<1 0 0>}
      localJointOffset=<4 0 0>
     }]
    relativeSpaceBone {boneName=base_hoodfront_r}
   }
   n4 = Anim Dynamics - Bone: dyn_hoodside_l(ComponentPose=@n3, Alpha=1, ExternalForce="0,0,0") {
    angularDampingOverride=1
    angularSpringConstant=8500
    bAngularSpring=true
    bOverrideAngularDamping=true
    bUseGravityOverride=true
    gravityScale=0.35
    simulationSpace=RootRelative
    alphaBoolBlend {blendOption=Linear}
    boundBone {boneName=dyn_hoodside_l}
    componentAppliedLinearAccClamp=<50000 50000 50000>
    componentLinearAccScale=<0.5 0.5 0.5>
    componentLinearVelScale=<0.5 0.5 0.5>
    gravityOverride {z=-980}
    physicsBodyDefinitions [{
      boundBone {boneName=dyn_hoodside_l}
      boxExtents=<20 20 20>
      constraintSetup {angularLimitsMax=<10 10 10>, angularLimitsMin=<-10 -10 -10>, angularTarget=<1 0 0>}
      localJointOffset=<5 0 0>
     }]
   }
   n5 = Anim Dynamics - Bone: dyn_hoodside_r(ComponentPose=@n4, Alpha=1, ExternalForce="0,0,0") {
    angularDampingOverride=1
    angularSpringConstant=8500
    bAngularSpring=true
    bOverrideAngularDamping=true
    bUseGravityOverride=true
    gravityScale=0.35
    simulationSpace=RootRelative
    alphaBoolBlend {blendOption=Linear}
    boundBone {boneName=dyn_hoodside_r}
    componentAppliedLinearAccClamp=<50000 50000 50000>
    componentLinearAccScale=<0.5 0.5 0.5>
    componentLinearVelScale=<0.5 0.5 0.5>
    gravityOverride {z=-980}
    physicsBodyDefinitions [{
      boundBone {boneName=dyn_hoodside_r}
      boxExtents=<20 20 20>
      constraintSetup {angularLimitsMax=<10 10 10>, angularLimitsMin=<-10 -10 -10>, angularTarget=<1 0 0>}
      localJointOffset=<5 0 0>
     }]
   }
   n6 = Anim Dynamics - Bone: dyn_hoodback(ComponentPose=@n5, Alpha=1, ExternalForce="0,0,0") {
    angularDampingOverride=1
    angularSpringConstant=9000
    bAngularSpring=true
    bOverrideAngularDamping=true
    bUseGravityOverride=true
    gravityScale=0.35
    simulationSpace=RootRelative
    alphaBoolBlend {blendOption=Linear}
    boundBone {boneName=dyn_hoodback}
    componentAppliedLinearAccClamp=<50000 50000 50000>
    componentLinearAccScale=<0.5 0.5 0.5>
    componentLinearVelScale=<0.5 0.5 0.5>
    gravityOverride {z=-980}
    physicsBodyDefinitions [{
      boundBone {boneName=dyn_hoodback}
      boxExtents=<20 20 20>
      constraintSetup {angularLimitsMax=<10 10 0>, angularLimitsMin=<-10 -10 0>, angularTarget=<1 0 0>}
      localJointOffset=<5 0 0>
     }]
   }
   Output Pose(Result=Component To Local(ComponentPose=@n6)) {layerGroup=None}
  ]}
 EventGraph {nodes [
   n1 = Event BlueprintProcessAnimInputEvent()
   > n2 = Branch(Condition=IsNamed(self=@n1.AnimInputEvent, InEventName=aie.Player.Zipline.Entered))
   > Set zip(zip:bool=true) -> (Output_Get:bool)
   Branch(execute=@n2.else, Condition=IsNamed(self=@n1.AnimInputEvent, InEventName=aie.Player.Zipline.Exited))
   > Set zip(zip=false)
   Event BlueprintUpdateAnimation()
   n3 = OR Boolean(A=Get bIsOnGrindRail(), B=Get zip() -> (zip:bool))
   > Set Phys_off(Phys_off:double="To Float (Boolean)"(InBool=NOT Boolean(A=@n3))) -> (Output_Get:double)
  ]}
]
variables [
 Phys_off {type=number, default=1}
 zip {type=boolean}
]
